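# UI material that draws a colored, skewed box glow for the Battle Royale HUD's player action alert.
# Rotated nine-slice UVs drive a position-dependent scale that skews a box SDF, which is anti-aliased and cubic-eased into the glow.
# A sine-based horizontal mask dims the center relative to the ends.
# Alpha-composite output scales opacity down by AdditiveMixRatio for a partly additive look, with Anim_Opacity as a fade control.
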Material /BattleRoyaleGameUI/HUD/PlayerActionAlert/Assets/M_UI_AngledBox_Glow
graph {
 properties {bAutomaticallySetUsageInEditor=false, blendMode=BLEND_AlphaComposite, materialDomain=MD_UI, usedShadingModels=Default Lit}
 nodes [
  n1 = MaterialFunctionCall!(UVs) {materialFunction=/FortUILibrary/MaterialFunctions/Basics/MF_UI_9SliceUVs}
  n2 = ConstantBiasScale(None=MaterialFunctionCall(UVs=@n1.UVs, "Rotation Angle (0-1)"=ScalarParameter() {defaultValue=0.25, parameterName=SkewRotation}) {materialFunction=&CustomRotator}.Rotated Values) {bias=-0.5, scale=2}
  n3 = MaterialFunctionCall(
   UVs=@n1.UVs
   Texture Scale=LinearInterpolate(B=ComponentMask(None=@n2) {r=true}, Alpha=ScalarParameter() {defaultValue=0.25, parameterName=SkewAmount}) {constA=1}
  ) {materialFunction=&ScaleUVsByCenter}
  Add(A=@n3.UVs)
  n4 = MaterialFunctionCall(
   UVs=@n3.UVs
   Size=ScalarParameter() {defaultValue=0.5, parameterName=Size}
   CornerRoundness=ScalarParameter() {parameterName=CornerRadius}
   KeepAspectRatio=StaticBool()
  ) {materialFunction=&MF_UI_Box}
  n5 = ScalarParameter() {parameterName=Threshold}
  n6 = ScalarParameter() {defaultValue=1, parameterName=Smoothing}
  n7 = MaterialFunctionCall("Value (0-1)"=Saturate(None=MaterialFunctionCall(DistanceField=@n4.SDF, Threshold=@n5, Smoothing=@n6) {materialFunction=&MF_UI_AntiAliasing}.Inner)) {materialFunction=&MF_UI_EaseCurves}
  n8 = Multiply(A=Multiply(A=Saturate(None=@n7.EaseInCubic), B=NamedRerouteUsage()), B=ScalarParameter() {defaultValue=1, parameterName=Intensity})
  n9 = ScalarParameter() {defaultValue=0.5, parameterName=AdditiveMixRatio}
  LinearInterpolate(B=Multiply(A=@n8, B=OneMinus(None=@n9)), Alpha=ScalarParameter() {defaultValue=1, parameterName=Anim_Opacity})
  n10 = VectorParameter() {parameterName=Color, defaultValue=<rgba 1 0 0 1>}
  Multiply(A=VertexColor().A, B=LinearInterpolate(B=Multiply(A=@n10.RGB, B=@n8), Alpha=ScalarParameter() {defaultValue=1, parameterName=Anim_Opacity}))
  n11 = Sine(None=MaterialFunctionCall() {materialFunction=&GetUserInterfaceUV}.Normalized UV) {period=2}
  n12 = MaterialFunctionCall("Value (0-1)"=OneMinus(None=ComponentMask(None=@n11) {r=true})) {materialFunction=&MF_UI_EaseCurves}
  NamedRerouteDeclaration(None=LinearInterpolate(A=ScalarParameter() {defaultValue=0.125, parameterName=FadeMin}, Alpha=Saturate(None=@n12.EaseInCubic))) {name=FadeMask, nodeColor=<rgba 1 0 0.6094 1>}
 ]
}
