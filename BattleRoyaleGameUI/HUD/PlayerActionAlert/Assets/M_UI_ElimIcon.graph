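# Translucent UI material that draws the elimination indicator icon in the HUD, with animation parameters for scale, color, alpha, threshold and smoothing.
# The center-scaled texture's green channel serves as a distance field for an antialiased shape and glow, with a box mask for optional filled eyes.
# Static switches choose adaptive, custom or shared-collection red coloring, glow-only and alpha-composite output.
# A sine-driven horizontal shake offset scales partly with viewport size.
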Material /BattleRoyaleGameUI/HUD/PlayerActionAlert/Assets/M_UI_ElimIcon
graph {
 properties {bAutomaticallySetUsageInEditor=false, blendMode=BLEND_Translucent, materialDomain=MD_UI, usedShadingModels=Default Lit}
 nodes [
  n1 = MaterialFunctionCall!(BufferSource) {materialFunction=/BattleRoyaleGameUI/Shared/MF_UI_AdaptiveColors}
  LinearInterpolate(A=@n1.Dark, B=@n1.Dark)
  n2 = WorldPosition()
  n3 = MaterialFunctionCall(Float2=@n2.XY) {materialFunction=&BreakOutFloat2Components}
  n4 = ScalarParameter() {defaultValue=32, parameterName=ShakeDistance}
  n5 = LinearInterpolate(
   A=Multiply(A=@n4, B=MaterialFunctionCall() {materialFunction=/FortUILibrary/MaterialFunctions/Basics/MF_UI_ViewportScale}.Viewport Scale Factor)
   B=Multiply(A=@n4) {constB=-1}
   Alpha=MaterialFunctionCall(Speed=ScalarParameter() {defaultValue=4, parameterName=ShakeSpeed}) {materialFunction=&DebugTimeSine}.Result
  ) {constA=40, constB=-40}
  n6 = LinearInterpolate(A=@n5, B=Multiply(A=@n5) {constB=-1}, Alpha=ScalarParameter() {parameterName=InvertShake})
  LinearInterpolate(A=@n2.XY, B=AppendVector(A=Add(A=@n3.R, B=@n6), B=@n3.G), Alpha=ScalarParameter() {parameterName=ScreenDistortionAmount})
  n7 = LinearInterpolate(
   A=ScalarParameter() {defaultValue=1.25, parameterName=Scale_Start}
   B=ScalarParameter() {defaultValue=1, parameterName=Scale_Stop}
   Alpha=ScalarParameter() {parameterName=Anim_Scale}
  )
  NamedRerouteDeclaration(None=TextureSample(
    UVs=MaterialFunctionCall(Texture Scale=@n7) {materialFunction=&ScaleUVsByCenter}.UVs
    Tex=TextureObjectParameter() {parameterName=TextureIcon, texture=/Game/Athena/HUD/Materials/T_UI_ElimIndicator}
   ) {samplerSource=SSM_Clamp_WorldGroupSettings, texture=&T_UI_ElimIndicator}.G) {name=IconMask, nodeColor=<rgba 0 0 0 1>}
  n8 = ScalarParameter() {defaultValue=1, group=Color, parameterName=AdaptiveColorIndex}
  n9 = LinearInterpolate(A=@n1.Dark, B=@n1.Dark, Alpha=@n8)
  n10 = Saturate(None=Subtract(A=@n8))
  n11 = Subtract(A=@n8) {constB=2}
  NamedRerouteDeclaration(None=LinearInterpolate(A=LinearInterpolate(A=@n9, B=@n1.Medium, Alpha=@n10), B=@n1.Bright, Alpha=Saturate(None=@n11))) {name=AdaptiveColorOutput, nodeColor=<rgba 0 0 0 1>}
  n12 = Constant()
  n13 = ScalarParameter() {defaultValue=0.025, parameterName=EyesMaskY}
  n14 = MaterialFunctionCall(
   UVs=Add(A=TextureCoordinate(), B=AppendVector(A=@n12, B=@n13))
   Size=AppendVector(A=ScalarParameter() {defaultValue=0.55, parameterName=EyesMaskWidth}, B=ScalarParameter() {defaultValue=0.35, parameterName=EyesMaskHeight})
  ) {materialFunction=&MF_UI_Box}
  n15 = NamedRerouteUsage()
  NamedRerouteDeclaration(None=Multiply(A=Saturate(None=@n14.Fill), B=OneMinus(None=@n15))) {name=EyesMask, nodeColor=<rgba 0 0.1094 1 1>}
  n16 = Multiply(
   A=ComponentMask(None=CollectionParameter() {collection=/BattleRoyaleGameUI/Shared/MPC_BRGameUI, parameterName=Red}) {b=true, g=true, r=true}
   B=ScalarParameter() {defaultValue=1, parameterName=DimColor}
  )
  n17 = VectorParameter() {parameterName=CustomColor, defaultValue=<rgba 1 1 1 1>}
  n18 = LinearInterpolate(A=@n16, B=@n17.RGB, Alpha=ScalarParameter() {group=Color, parameterName=Anim_Color})
  n19 = NamedRerouteUsage()
  n20 = LinearInterpolate(A=@n19, B=@n1.Dark, Alpha=ScalarParameter() {group=Color, parameterName=Anim_Color})
  n21 = StaticSwitchParameter(True=@n17.RGB, False=@n16) {parameterName=CustomColor}
  n22 = NamedRerouteDeclaration(None=StaticSwitchParameter(
    True=StaticSwitchParameter(True=@n18, False=@n20) {parameterName=CustomColor_Anim}
    False=StaticSwitchParameter(True=@n19, False=@n21) {defaultValue=true, parameterName=AdaptiveColor}
   ) {defaultValue=true, parameterName=AnimatedColors}) {name=Color, nodeColor=<rgba 0 0 0 1>}
  n23 = LinearInterpolate(B=NamedRerouteUsage(), Alpha=ScalarParameter() {defaultValue=1, parameterName=Anim_Alpha})
  n24 = NamedRerouteUsage()
  StaticSwitchParameter(
   True=Multiply(A=VertexColor().A, B=Multiply(A=@n22, B=@n23))
   False=LinearInterpolate(A=@n22, B=LinearInterpolate(A=@n22, B=@n1.Dark, Alpha=@n24), Alpha=ScalarParameter() {parameterName=IsEyesFilled})
  ) {parameterName=IsAlphaCompositeBlendMode}
  n25 = ScalarParameter() {parameterName=AdditiveRatio}
  n26 = MaterialFunctionCall(
   DistanceField=NamedRerouteUsage()
   Threshold=ScalarParameter() {defaultValue=0.25, parameterName=GlowThreshold}
   Smoothing=ScalarParameter() {defaultValue=0.5, parameterName=GlowSmoothing}
  ) {materialFunction=&MF_UI_AntiAliasing}
  n27 = MaterialFunctionCall("Value (0-1)"=@n26.Outer) {materialFunction=&MF_UI_EaseCurves}
  n28 = ScalarParameter() {parameterName=UseGlowEaseCurveAdjsutment}
  n29 = Multiply(A=Saturate(None=LinearInterpolate(A=@n26.Outer, B=@n27.EaseInCubic, Alpha=@n28)), B=ScalarParameter() {parameterName=GlowAmount})
  n30 = ScalarParameter() {defaultValue=0.25, parameterName=Threshold_Start}
  n31 = ScalarParameter() {defaultValue=0.5, parameterName=Threshold_Stop}
  n32 = ScalarParameter() {parameterName=Anim_Threshold}
  n33 = ScalarParameter() {parameterName=Smoothing_Start}
  n34 = ScalarParameter() {defaultValue=1, parameterName=Smoothing_Stop}
  n35 = ScalarParameter() {parameterName=Anim_Smoothing}
  n36 = Saturate(None=MaterialFunctionCall(
    DistanceField=NamedRerouteUsage()
    Threshold=LinearInterpolate(A=@n30, B=@n31, Alpha=@n32) {constA=0.25}
    Smoothing=LinearInterpolate(A=@n33, B=@n34, Alpha=@n35)
   ) {materialFunction=&MF_UI_AntiAliasing}.Outer)
  n37 = StaticSwitchParameter(True=@n29, False=LinearInterpolate(A=@n29, Alpha=NamedRerouteDeclaration(None=@n36) {name=IconMask_2, nodeColor=<rgba 1 0.2109 0 1>})) {parameterName=GlowOnly}
  n38 = ScalarParameter() {defaultValue=1, parameterName=Opacity}
  n39 = LinearInterpolate(
   B=LinearInterpolate(A=@n38, B=@n38, Alpha=ScalarParameter() {parameterName=HaslLessThanOneOpacity})
   Alpha=ScalarParameter() {defaultValue=1, parameterName=Anim_Alpha}
  )
  n40 = LinearInterpolate(A=@n37, B=Constant() {r=1}, Alpha=NamedRerouteUsage())
  n41 = ScalarParameter() {parameterName=IsEyesFilled}
  StaticSwitchParameter(True=Multiply(A=Multiply(A=@n37, B=@n39), B=OneMinus(None=@n25)), False=Multiply(A=LinearInterpolate(A=@n37, B=@n40, Alpha=@n41), B=@n39)) {parameterName=IsAlphaCompositeBlendMode}
 ]
}
