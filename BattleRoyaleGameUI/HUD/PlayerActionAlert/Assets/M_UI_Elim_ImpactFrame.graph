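# Translucent UI material that draws an animated impact frame shape, apparently for the elimination player action alert.
# Anim_Overall_Scale scales the UVs about the center of a bolt backplate distance field texture.
# MF_UI_AntiAliasing cuts a smooth-edged mask at a threshold driven by Anim_DF_Scale, which Anim_Alpha fades out as it rises.
# The Charcoal color comes from the shared MPC_BRGameUI parameter collection.
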Material /BattleRoyaleGameUI/HUD/PlayerActionAlert/Assets/M_UI_Elim_ImpactFrame
graph {
 properties {bAutomaticallySetUsageInEditor=false, blendMode=BLEND_Translucent, materialDomain=MD_UI, usedShadingModels=Default Lit}
 nodes [
  ComponentMask(None=CollectionParameter() {collection=/BattleRoyaleGameUI/Shared/MPC_BRGameUI, parameterName=Charcoal}) {b=true, g=true, r=true}
  n1 = TextureCoordinate()
  n2 = LinearInterpolate(Alpha=ScalarParameter() {parameterName=Anim_Overall_Scale}) {constA=1, constB=0.5}
  n3 = Add(A=TextureSample(UVs=MaterialFunctionCall(UVs=@n1, Texture Scale=@n2) {materialFunction=&ScaleUVsByCenter}.UVs) {samplerSource=SSM_Clamp_WorldGroupSettings, texture=/Game/Athena/HUD/Tricks/Materials/T_UI_BoltBackplate2_DF}.R) {constB=0.025}
  n4 = LinearInterpolate(Alpha=ScalarParameter() {parameterName=Anim_DF_Scale}) {constA=0.25, constB=0.5}
  n5 = Constant() {r=0.01}
  Multiply(
   A=Saturate(None=MaterialFunctionCall(DistanceField=@n3, Threshold=@n4, Smoothing=@n5) {materialFunction=&MF_UI_AntiAliasing}.Outer)
   B=LinearInterpolate(Alpha=ScalarParameter() {parameterName=Anim_Alpha}) {constA=1, constB=0}
  )
 ]
}
